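# Column header row widget for the Athena replay browser, labelling the columns of the replay list.
# A fixed-height SizeBox wraps a CommonBorder and a horizontal box, with each label in its own SizeBox column.
# Name, Date, Length, Finish and Eliminations use fixed widths, while Version, Size, Mode and Views have minimum widths and are collapsed.
# It is purely static layout on CommonUserWidget, with an empty event graph and no variables.
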WidgetBlueprint /FortniteCoreUI/Athena/UI/ReplayBrowser/AthenaReplayBrowserHeaderRow
graphs [EventGraph {links [], nodes []}]
variables []
widgets {
 info {inheritedWidgetCount=0, parentClass=&CommonUserWidget, rootWidgetClass=&SizeBox, widgetCount=25}
 widgets [
  SizeBox_0 {widgetClassPath=&SizeBox, widget=SizeBox_0, properties {bOverride_HeightOverride=true, heightOverride=63.5769}}
  CommonBorder_0 {
   widgetClassPath=&CommonBorder
   parent=SizeBox_0
   widget=CommonBorder_0
   properties {padding=<ltrb 0 0 0 0>}
   slot {class=SizeBoxSlot}
  }
  HorizontalBox_0 {
   widgetClassPath=&HorizontalBox
   parent=CommonBorder_0
   widget=HorizontalBox_0
   slot {class=BorderSlot, properties {padding=<ltrb 0 0 0 0>}}
  }
  NameSB {
   widgetClassPath=&SizeBox
   parent=HorizontalBox_0
   widget=NameSB
   properties {bOverride_WidthOverride=true, widthOverride=720}
   slot {class=HorizontalBoxSlot, properties {padding=<ltrb 0 0 8 0>}}
  }
  HorizontalBox_1 {
   widgetClassPath=&HorizontalBox
   parent=NameSB
   widget=HorizontalBox_1
   slot {class=SizeBoxSlot, properties {verticalAlignment=VAlign_Center, padding=<ltrb 32 4 0 4>}}
  }
  Name {
   widgetClassPath=&CommonTextBlock
   parent=HorizontalBox_1
   widget=Name
   properties {
    scrollStyle=&TextScrollStyle-Base_C
    text=NAME
    font {
     bForceMonospaced=false
     bMaterialIsStencil=false
     fontMaterial=null
     fontObject=&BurbankSmall
     letterSpacing=0
     monospacedWidth=1
     size=28
     skewAmount=0
     typefaceFontName=Black
     outlineSettings {
      bApplyOutlineToDropShadows=false
      bMiteredCorners=false
      bSeparateFillAlpha=false
      outlineMaterial=null
      outlineSize=0
      outlineColor=<rgba 0 0 0 1>
     }
    }
    shadowOffset=<0 0>
   }
   slot {class=HorizontalBoxSlot, properties {size {sizeRule=Fill, value=1}}}
  }
  DateSB {
   widgetClassPath=&SizeBox
   parent=HorizontalBox_0
   widget=DateSB
   properties {bOverride_WidthOverride=true, widthOverride=256}
   slot {class=HorizontalBoxSlot, properties {padding=<ltrb 8 0 8 0>}}
  }
  Date {
   widgetClassPath=&CommonTextBlock
   parent=DateSB
   widget=Date
   properties {
    justification=Center
    scrollStyle=&TextScrollStyle-Base_C
    text=DATE
    font {
     bForceMonospaced=false
     bMaterialIsStencil=false
     fontMaterial=null
     fontObject=&BurbankSmall
     letterSpacing=0
     monospacedWidth=1
     size=20
     skewAmount=0
     typefaceFontName=Bold
     outlineSettings {
      bApplyOutlineToDropShadows=false
      bMiteredCorners=false
      bSeparateFillAlpha=false
      outlineMaterial=null
      outlineSize=0
      outlineColor=<rgba 0 0 0 1>
     }
    }
    shadowOffset=<0 0>
   }
   slot {class=SizeBoxSlot, properties {verticalAlignment=VAlign_Center, padding=<ltrb 5 4 5 4>}}
  }
  VersionSB {
   widgetClassPath=&SizeBox
   parent=HorizontalBox_0
   widget=VersionSB
   properties {bOverride_MinDesiredWidth=true, minDesiredWidth=90, visibility=Collapsed}
   slot {class=HorizontalBoxSlot, properties {padding=<ltrb 8 0 8 0>}}
  }
  HorizontalBox_3 {
   widgetClassPath=&HorizontalBox
   parent=VersionSB
   widget=HorizontalBox_3
   slot {class=SizeBoxSlot, properties {verticalAlignment=VAlign_Center, padding=<ltrb 5 4 5 4>}}
  }
  Version {
   widgetClassPath=&CommonTextBlock
   parent=HorizontalBox_3
   widget=Version
   properties {
    scrollStyle=&TextScrollStyle-Base_C
    style=/FortUILibrary/Styles/Typography/Foundation/TextStyle-Base-XS-B.TextStyle-Base-XS-B_C
    text=VER.
    font {
     bForceMonospaced=false
     bMaterialIsStencil=false
     fontMaterial=null
     fontObject=&NotoSans
     letterSpacing=0
     monospacedWidth=1
     size=16
     skewAmount=0
     typefaceFontName=NotoSans-Bold
     outlineSettings {
      bApplyOutlineToDropShadows=false
      bMiteredCorners=false
      bSeparateFillAlpha=false
      outlineMaterial=null
      outlineSize=0
      outlineColor=<rgba 0 0 0 1>
     }
    }
    shadowOffset=<0 0>
   }
   slot {class=HorizontalBoxSlot, properties {horizontalAlignment=HAlign_Left, size {sizeRule=Fill, value=1}}}
  }
  SizeSb {
   widgetClassPath=&SizeBox
   parent=HorizontalBox_0
   widget=SizeSb
   properties {bOverride_MinDesiredWidth=true, minDesiredWidth=105, visibility=Collapsed}
   slot {class=HorizontalBoxSlot, properties {padding=<ltrb 8 0 8 0>}}
  }
  HorizontalBox_4 {
   widgetClassPath=&HorizontalBox
   parent=SizeSb
   widget=HorizontalBox_4
   slot {class=SizeBoxSlot, properties {verticalAlignment=VAlign_Center, padding=<ltrb 5 4 5 4>}}
  }
  Size {
   widgetClassPath=&CommonTextBlock
   parent=HorizontalBox_4
   widget=Size
   properties {
    scrollStyle=&TextScrollStyle-Base_C
    style=&TextStyle-Base-XS-B_C
    text=SIZE
    font {
     bForceMonospaced=false
     bMaterialIsStencil=false
     fontMaterial=null
     fontObject=&NotoSans
     letterSpacing=0
     monospacedWidth=1
     size=16
     skewAmount=0
     typefaceFontName=NotoSans-Bold
     outlineSettings {
      bApplyOutlineToDropShadows=false
      bMiteredCorners=false
      bSeparateFillAlpha=false
      outlineMaterial=null
      outlineSize=0
      outlineColor=<rgba 0 0 0 1>
     }
    }
    shadowOffset=<0 0>
   }
   slot {class=HorizontalBoxSlot, properties {horizontalAlignment=HAlign_Left, size {sizeRule=Fill, value=1}}}
  }
  ModeSB {
   widgetClassPath=&SizeBox
   parent=HorizontalBox_0
   widget=ModeSB
   properties {bOverride_MinDesiredWidth=true, minDesiredWidth=150, visibility=Collapsed}
   slot {class=HorizontalBoxSlot, properties {padding=<ltrb 8 0 8 0>}}
  }
  HorizontalBox_5 {
   widgetClassPath=&HorizontalBox
   parent=ModeSB
   widget=HorizontalBox_5
   slot {class=SizeBoxSlot, properties {verticalAlignment=VAlign_Center, padding=<ltrb 5 4 5 4>}}
  }
  Mode {
   widgetClassPath=&CommonTextBlock
   parent=HorizontalBox_5
   widget=Mode
   properties {
    scrollStyle=&TextScrollStyle-Base_C
    style=&TextStyle-Base-XS-B_C
    text=MODE
    font {
     bForceMonospaced=false
     bMaterialIsStencil=false
     fontMaterial=null
     fontObject=&NotoSans
     letterSpacing=0
     monospacedWidth=1
     size=16
     skewAmount=0
     typefaceFontName=NotoSans-Bold
     outlineSettings {
      bApplyOutlineToDropShadows=false
      bMiteredCorners=false
      bSeparateFillAlpha=false
      outlineMaterial=null
      outlineSize=0
      outlineColor=<rgba 0 0 0 1>
     }
    }
    shadowOffset=<0 0>
   }
   slot {class=HorizontalBoxSlot, properties {horizontalAlignment=HAlign_Left, size {sizeRule=Fill, value=1}}}
  }
  LengthSB {
   widgetClassPath=&SizeBox
   parent=HorizontalBox_0
   widget=LengthSB
   properties {bOverride_WidthOverride=true, widthOverride=200}
   slot {class=HorizontalBoxSlot, properties {padding=<ltrb 8 0 8 0>}}
  }
  Length {
   widgetClassPath=&CommonTextBlock
   parent=LengthSB
   widget=Length
   properties {
    justification=Center
    scrollStyle=&TextScrollStyle-Base_C
    text=LENGTH
    font {
     bForceMonospaced=false
     bMaterialIsStencil=false
     fontMaterial=null
     fontObject=&BurbankSmall
     letterSpacing=0
     monospacedWidth=1
     size=20
     skewAmount=0
     typefaceFontName=Bold
     outlineSettings {
      bApplyOutlineToDropShadows=false
      bMiteredCorners=false
      bSeparateFillAlpha=false
      outlineMaterial=null
      outlineSize=0
      outlineColor=<rgba 0 0 0 1>
     }
    }
    shadowOffset=<0 0>
   }
   slot {class=SizeBoxSlot, properties {verticalAlignment=VAlign_Center, padding=<ltrb 5 4 5 4>}}
  }
  FinishSB {
   widgetClassPath=&SizeBox
   parent=HorizontalBox_0
   widget=FinishSB
   properties {bOverride_WidthOverride=true, widthOverride=200}
   slot {class=HorizontalBoxSlot, properties {padding=<ltrb 8 0 8 0>}}
  }
  Finish {
   widgetClassPath=&CommonTextBlock
   parent=FinishSB
   widget=Finish
   properties {
    justification=Center
    scrollStyle=&TextScrollStyle-Base_C
    text=FINISH
    font {
     bForceMonospaced=false
     bMaterialIsStencil=false
     fontMaterial=null
     fontObject=&BurbankSmall
     letterSpacing=0
     monospacedWidth=1
     size=20
     skewAmount=0
     typefaceFontName=Bold
     outlineSettings {
      bApplyOutlineToDropShadows=false
      bMiteredCorners=false
      bSeparateFillAlpha=false
      outlineMaterial=null
      outlineSize=0
      outlineColor=<rgba 0 0 0 1>
     }
    }
    shadowOffset=<0 0>
   }
   slot {class=SizeBoxSlot, properties {verticalAlignment=VAlign_Center, padding=<ltrb 5 4 5 4>}}
  }
  KillsSB {
   widgetClassPath=&SizeBox
   parent=HorizontalBox_0
   widget=KillsSB
   properties {bOverride_WidthOverride=true, widthOverride=200}
   slot {class=HorizontalBoxSlot, properties {padding=<ltrb 8 0 8 0>}}
  }
  Kills {
   widgetClassPath=&CommonTextBlock
   parent=KillsSB
   widget=Kills
   properties {
    justification=Center
    scrollStyle=&TextScrollStyle-Base_C
    text=ELIMINATIONS
    font {
     bForceMonospaced=false
     bMaterialIsStencil=false
     fontMaterial=null
     fontObject=&BurbankSmall
     letterSpacing=0
     monospacedWidth=1
     size=20
     skewAmount=0
     typefaceFontName=Bold
     outlineSettings {
      bApplyOutlineToDropShadows=false
      bMiteredCorners=false
      bSeparateFillAlpha=false
      outlineMaterial=null
      outlineSize=0
      outlineColor=<rgba 0 0 0 1>
     }
    }
    shadowOffset=<0 0>
   }
   slot {class=SizeBoxSlot, properties {verticalAlignment=VAlign_Center, padding=<ltrb 5 4 5 4>}}
  }
  ViewsSB {
   widgetClassPath=&SizeBox
   parent=HorizontalBox_0
   widget=ViewsSB
   properties {bOverride_MinDesiredWidth=true, minDesiredWidth=85, visibility=Collapsed}
   slot {class=HorizontalBoxSlot, properties {padding=<ltrb 8 0 8 0>}}
  }
  Views {
   widgetClassPath=&CommonTextBlock
   parent=ViewsSB
   widget=Views
   properties {
    scrollStyle=&TextScrollStyle-Base_C
    style=&TextStyle-Base-XS-B_C
    text=VIEWS
    font {
     bForceMonospaced=false
     bMaterialIsStencil=false
     fontMaterial=null
     fontObject=&NotoSans
     letterSpacing=0
     monospacedWidth=1
     size=16
     skewAmount=0
     typefaceFontName=NotoSans-Bold
     outlineSettings {
      bApplyOutlineToDropShadows=false
      bMiteredCorners=false
      bSeparateFillAlpha=false
      outlineMaterial=null
      outlineSize=0
      outlineColor=<rgba 0 0 0 1>
     }
    }
    shadowOffset=<0 0>
   }
   slot {class=SizeBoxSlot, properties {verticalAlignment=VAlign_Center, padding=<ltrb 5 4 5 4>}}
  }
 ]
}
