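# Modal dialog widget for renaming a saved replay in the Athena replay browser.
# A blurred background and Lightbox host a full-screen tap-to-close button behind a centered box with title, prompt, editable name field and Confirm/Decline buttons.
# The Blueprint has an empty event graph, leaving behavior to native parent FortReplayBrowserRenameDialog, which supplies the text field and three buttons as inherited widgets.
# Confirm and Decline are non-focusable and triggered through input-action table rows.
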WidgetBlueprint /FortniteCoreUI/Athena/UI/ReplayBrowser/AthenaReplayBrowserRenameDialog
graphs [EventGraph {links [], nodes []}]
variables []
widgets {
 info {
  inheritedWidgetCount=4
  namedSlotCount=1
  parentClass=/Script/FortniteUI.FortReplayBrowserRenameDialog
  rootWidgetClass=&BackgroundBlur
  widgetCount=16
 }
 widgets [
  BackgroundBlurWidget_0 {widgetClassPath=&BackgroundBlur, widget=BackgroundBlurWidget_0, properties {blurStrength=15}}
  Lightbox_NUI {widgetClassPath=/Game/UI/Foundation/Lightbox/Lightbox, parent=BackgroundBlurWidget_0, widget=Lightbox_NUI, slot {class=BackgroundBlurSlot}}
  Overlay_3 {widgetClassPath=&Overlay, namedSlotHost=Lightbox_NUI, widget=Overlay_3}
  Image_0 {
   widgetClassPath=&Image
   parent=Overlay_3
   widget=Image_0
   properties {
    visibility=SelfHitTestInvisible
    brush {
     bIsDynamicallyLoaded=false
     drawAs=Image
     imageType=NoImage
     mirroring=NoMirror
     resourceName=None
     resourceObject=/FortUILibrary/Styles/Border/Materials/M_Border_SoftShadow_DarkBLue
     tiling=NoTile
     imageSize=<32 32>
     margin=<ltrb 0 0 0 0>
     outlineSettings {
      bUseBrushTransparency=false
      roundingType=HalfHeightRadius
      width=0
      color {colorUseRule=UseColor_Specified, specifiedColor=<rgba 0 0 0 0>}
      cornerRadii=<0 0 0 0>
     }
     tintColor {colorUseRule=UseColor_Specified, specifiedColor=<rgba 1 1 1 1>}
     uVRegion {bIsValid=false, max=<0 0>, min=<0 0>}
    }
    colorAndOpacity=<rgba 1 1 1 0.33>
   }
   slot {class=OverlaySlot, properties {horizontalAlignment=HAlign_Fill, verticalAlignment=VAlign_Fill, padding=<ltrb 256 128 256 128>}}
  }
  Button_TapToClose {
   widgetClassPath=/Game/Athena/UI/Matchmaking/SimpleCommonButton
   bInherited=true
   parent=Overlay_3
   widget=Button_TapToClose
   properties {
    bApplyAlphaOnDisable=false
    bDisplayInputActionWhenNotInteractable=false
    bShouldUseFallbackDefaultInputAction=false
    onButtonBaseClicked=(&self:WidgetTree.Button_TapToClose.HandleOnButtonClicked)
    onButtonBaseDoubleClicked=(&self:WidgetTree.Button_TapToClose.HandleOnButtonDoubleClicked)
    onButtonBaseHovered=(&self:WidgetTree.Button_TapToClose.HandleOnButtonHovered)
    onButtonBaseUnhovered=(&self:WidgetTree.Button_TapToClose.HandleOnButtonUnhovered)
    onSelectedChangedBase=(&self:WidgetTree.Button_TapToClose.HandleOnSelectedChanged)
    style=/FortUILibrary/Foundation/Buttons/ButtonStyles/Athena/ButtonStyle_TransparentList_Nothing.ButtonStyle_TransparentList_Nothing_C
   }
   slot {class=OverlaySlot, properties {horizontalAlignment=HAlign_Fill, verticalAlignment=VAlign_Fill}}
  }
  SizeBox_0 {
   widgetClassPath=&SizeBox
   parent=Overlay_3
   widget=SizeBox_0
   properties {bOverride_HeightOverride=true, bOverride_WidthOverride=true, heightOverride=300, widthOverride=720}
   slot {class=OverlaySlot, properties {horizontalAlignment=HAlign_Center, verticalAlignment=VAlign_Center, padding=<ltrb 0 200 0 200>}}
  }
  MainBox {
   widgetClassPath=&VerticalBox
   parent=SizeBox_0
   widget=MainBox
   properties {navigation=&self:WidgetTree.MainBox.WidgetNavigation_0}
   slot {class=SizeBoxSlot}
  }
  CommonBorder_Modal {
   widgetClassPath=&CommonBorder
   parent=MainBox
   widget=CommonBorder_Modal
   properties {horizontalAlignment=HAlign_Center}
   slot {class=VerticalBoxSlot}
  }
  Title {
   widgetClassPath=&CommonTextBlock
   parent=CommonBorder_Modal
   widget=Title
   properties {
    clipping=ClipToBounds
    scrollStyle=&TextScrollStyle-Base_C
    style=/FortUILibrary/Styles/Typography/Foundation/TextStyle-Header-L.TextStyle-Header-L_C
    text=Rename Replay
    textTransformPolicy=ToUpper
    font {
     bForceMonospaced=false
     bMaterialIsStencil=false
     fontMaterial=null
     fontObject=&Burbank
     letterSpacing=0
     monospacedWidth=1
     size=48
     skewAmount=0
     typefaceFontName=Burbank-Black
     outlineSettings {
      bApplyOutlineToDropShadows=false
      bMiteredCorners=false
      bSeparateFillAlpha=false
      outlineMaterial=null
      outlineSize=0
      outlineColor=<rgba 0 0 0 1>
     }
    }
    shadowOffset=<0 0>
   }
   slot {class=BorderSlot, properties {horizontalAlignment=HAlign_Center}}
  }
  TextItemContent {
   widgetClassPath=&VerticalBox
   parent=MainBox
   widget=TextItemContent
   slot {class=VerticalBoxSlot, properties {padding=<ltrb 0 20 0 0>, size {sizeRule=Fill, value=1}}}
  }
  Description {
   widgetClassPath=&CommonTextBlock
   parent=TextItemContent
   widget=Description
   properties {
    autoWrapText=true
    justification=Center
    text=Enter new replay name:
    colorAndOpacity {colorUseRule=UseColor_Specified, specifiedColor=<rgba 0.0331 0.855 1 1>}
    font {
     bForceMonospaced=false
     bMaterialIsStencil=false
     fontMaterial=null
     fontObject=&BurbankSmall
     letterSpacing=0
     monospacedWidth=1
     size=16
     skewAmount=0
     typefaceFontName=Bold
     outlineSettings {
      bApplyOutlineToDropShadows=false
      bMiteredCorners=false
      bSeparateFillAlpha=false
      outlineMaterial=null
      outlineSize=0
      outlineColor=<rgba 0 0 0 1>
     }
    }
    shadowOffset=<0 0>
   }
   slot {class=VerticalBoxSlot}
  }
  CommonBorder_1 {
   widgetClassPath=&CommonBorder
   parent=TextItemContent
   widget=CommonBorder_1
   properties {
    background {
     bIsDynamicallyLoaded=false
     drawAs=Box
     imageType=NoImage
     mirroring=NoMirror
     resourceName=None
     resourceObject=/Game/Athena/UI/ReplayBrowser/M_Ui_Skewed_Vfor9slices
     tiling=NoTile
     imageSize=<32 32>
     margin=<ltrb 0.5 0 0.5 0>
     outlineSettings {
      bUseBrushTransparency=false
      roundingType=HalfHeightRadius
      width=0
      color {colorUseRule=UseColor_Specified, specifiedColor=<rgba 0 0 0 0>}
      cornerRadii=<0 0 0 0>
     }
     tintColor {colorUseRule=UseColor_Specified, specifiedColor=<rgba 1 1 1 1>}
     uVRegion {bIsValid=false, max=<0 0>, min=<0 0>}
    }
    padding=<ltrb 12 12 12 12>
   }
   slot {class=VerticalBoxSlot, properties {padding=<ltrb 0 10 0 8>}}
  }
  TextBox_Name {
   widgetClassPath=/Script/UMG.EditableText
   bInherited=true
   parent=CommonBorder_1
   widget=TextBox_Name
   properties {
    hintText=New Name
    navigation=&self:WidgetTree.TextBox_Name.WidgetNavigation_0
    widgetStyle {
     backgroundImageComposing {
      bIsDynamicallyLoaded=false
      drawAs=Border
      imageType=FullColor
      mirroring=NoMirror
      resourceName=../../../Engine/Content/Slate/Old/HyperlinkDotted.png
      resourceObject=null
      tiling=Both
      imageSize=<0 0>
      margin=<ltrb 0 0 0 0.1875>
      outlineSettings {
       bUseBrushTransparency=false
       roundingType=HalfHeightRadius
       width=0
       color {colorUseRule=UseColor_Specified, specifiedColor=<rgba 0 0 0 0>}
       cornerRadii=<0 0 0 0>
      }
      tintColor {colorUseRule=UseColor_Specified, specifiedColor=<rgba 1 1 1 1>}
      uVRegion {bIsValid=false, max=<0 0>, min=<0 0>}
     }
     backgroundImageSelected {
      bIsDynamicallyLoaded=false
      drawAs=Box
      imageType=FullColor
      mirroring=NoMirror
      resourceName=../../../Engine/Content/Slate/Common/EditableTextSelectionBackground.png
      resourceObject=null
      tiling=NoTile
      imageSize=<0 0>
      margin=<ltrb 0.25 0.25 0.25 0.25>
      outlineSettings {
       bUseBrushTransparency=false
       roundingType=HalfHeightRadius
       width=0
       color {colorUseRule=UseColor_Specified, specifiedColor=<rgba 0 0 0 0>}
       cornerRadii=<0 0 0 0>
      }
      tintColor {colorUseRule=UseColor_Specified, specifiedColor=<rgba 1 1 1 1>}
      uVRegion {bIsValid=false, max=<0 0>, min=<0 0>}
     }
     caretImage {
      bIsDynamicallyLoaded=false
      drawAs=Image
      imageType=FullColor
      mirroring=NoMirror
      resourceName=../../../Engine/Content/Slate/Old/White.png
      resourceObject=null
      tiling=NoTile
      imageSize=<16 16>
      margin=<ltrb 0 0 0 0>
      outlineSettings {
       bUseBrushTransparency=false
       roundingType=HalfHeightRadius
       width=0
       color {colorUseRule=UseColor_Specified, specifiedColor=<rgba 0 0 0 0>}
       cornerRadii=<0 0 0 0>
      }
      tintColor {colorUseRule=UseColor_Specified, specifiedColor=<rgba 1 1 1 1>}
      uVRegion {bIsValid=false, max=<0 0>, min=<0 0>}
     }
     colorAndOpacity {colorUseRule=UseColor_Specified, specifiedColor=<rgba 0.0075 0.0176 0.2086 1>}
     font {
      bForceMonospaced=false
      bMaterialIsStencil=false
      fontMaterial=null
      fontObject=&BurbankSmall
      letterSpacing=0
      monospacedWidth=1
      size=16
      skewAmount=0
      typefaceFontName=Bold
      outlineSettings {
       bApplyOutlineToDropShadows=false
       bMiteredCorners=false
       bSeparateFillAlpha=false
       outlineMaterial=null
       outlineSize=0
       outlineColor=<rgba 0 0 0 1>
      }
     }
     hintColorAndOpacity {colorUseRule=UseColor_Foreground, specifiedColor=<rgba 1 0 1 1>}
     selectedBackgroundColor {colorUseRule=UseColor_Foreground, specifiedColor=<rgba 1 0 1 1>}
    }
   }
   slot {class=BorderSlot, properties {padding=<ltrb 12 12 12 12>}}
  }
  ButtonBox {
   widgetClassPath=&HorizontalBox
   parent=MainBox
   widget=ButtonBox
   slot {class=VerticalBoxSlot, properties {horizontalAlignment=HAlign_Center, verticalAlignment=VAlign_Bottom, padding=<ltrb 20 20 20 20>}}
  }
  Button_Confirm {
   widgetClassPath=&IconTextButton
   bInherited=true
   parent=ButtonBox
   widget=Button_Confirm
   properties {
    bDisplayInputActionWhenNotInteractable=false
    bIsFocusable=false
    buttonText=Confirm
    minHeight=55
    minWidth=256
    onButtonBaseClicked=(&self:WidgetTree.Button_Confirm.HandleOnButtonClicked)
    onButtonBaseDoubleClicked=(&self:WidgetTree.Button_Confirm.HandleOnButtonDoubleClicked)
    onButtonBaseHovered=(&self:WidgetTree.Button_Confirm.HandleOnButtonHovered)
    onButtonBaseUnhovered=(&self:WidgetTree.Button_Confirm.HandleOnButtonUnhovered)
    onSelectedChangedBase=(&self:WidgetTree.Button_Confirm.HandleOnSelectedChanged)
    style=/FortUILibrary/Foundation/Buttons/ButtonStyles/Primary/ButtonStyle-Primary-M_Skew_Yellow.ButtonStyle-Primary-M_Skew_Yellow_C
    triggeringInputAction {dataTable=&InputActionDataTable, rowName=Confirm}
   }
   slot {class=HorizontalBoxSlot, properties {horizontalAlignment=HAlign_Center, verticalAlignment=VAlign_Center, padding=<ltrb 12 0 12 0>}}
  }
  Button_Decline {
   widgetClassPath=&IconTextButton
   bInherited=true
   parent=ButtonBox
   widget=Button_Decline
   properties {
    bIsFocusable=false
    buttonText=Decline
    minHeight=55
    minWidth=256
    onButtonBaseClicked=(&self:WidgetTree.Button_Decline.HandleOnButtonClicked)
    onButtonBaseDoubleClicked=(&self:WidgetTree.Button_Decline.HandleOnButtonDoubleClicked)
    onButtonBaseHovered=(&self:WidgetTree.Button_Decline.HandleOnButtonHovered)
    onButtonBaseUnhovered=(&self:WidgetTree.Button_Decline.HandleOnButtonUnhovered)
    onSelectedChangedBase=(&self:WidgetTree.Button_Decline.HandleOnSelectedChanged)
    style=/FortUILibrary/Foundation/Buttons/ButtonStyles/Primary/ButtonStyle-Primary-M_Skew_Blue.ButtonStyle-Primary-M_Skew_Blue_C
    triggeringInputAction {dataTable=&InputActionDataTable, rowName=Decline}
   }
   slot {class=HorizontalBoxSlot, properties {horizontalAlignment=HAlign_Center, verticalAlignment=VAlign_Center, padding=<ltrb 12 0 12 0>}}
  }
 ]
}
