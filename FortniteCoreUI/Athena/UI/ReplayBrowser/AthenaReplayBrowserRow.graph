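# Replay browser row widget showing one replay's name, date, length, finish rank and eliminations in fixed-width columns.
# The replay state event tints the background material's Status Color by saved state, with a separate color for old or corrupt replays.
# Hover and selection events play short animations forward or reversed, recoloring text and skewing the background via material parameters.
# Selection picks a transition from hovered or idle, and hover animations are skipped while selected.
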WidgetBlueprint /FortniteCoreUI/Athena/UI/ReplayBrowser/AthenaReplayBrowserRow
animations [
 Selected_Direct {
  binding_count=6
  length_seconds=0.1
  object=&self:Selected_Direct
  sequence {bindings [
    Text_Date {
     id=#1
     possessed_class=CommonTextBlock
     tracks [{
       type=MovieSceneColorTrack
       display_name=Color and Opacity
       sections [{
         type=MovieSceneColorSection
         end_frame=2
         start_frame=0
         channels [
          Color.R {type=MovieSceneScriptingFloatChannel, keys [[0,0.6038],[2,0.007]]}
          Color.G {type=MovieSceneScriptingFloatChannel, keys [[0,0.8879],[2,0.017]]}
          Color.B {type=MovieSceneScriptingFloatChannel, keys [[0,1],[2,0.2]]}
          Color.A {type=MovieSceneScriptingFloatChannel}
         ]
        }]
      }]
    }
    Text_Length {
     id=#2
     possessed_class=CommonTextBlock
     tracks [{
       type=MovieSceneColorTrack
       display_name=Color and Opacity
       sections [{
         type=MovieSceneColorSection
         end_frame=2
         start_frame=0
         channels [
          Color.R {type=MovieSceneScriptingFloatChannel, keys [[0,0.6038],[2,0.007]]}
          Color.G {type=MovieSceneScriptingFloatChannel, keys [[0,0.8879],[2,0.017]]}
          Color.B {type=MovieSceneScriptingFloatChannel, keys [[0,1],[2,0.2]]}
          Color.A {type=MovieSceneScriptingFloatChannel}
         ]
        }]
      }]
    }
    Text_Eliminations {
     id=#3
     possessed_class=CommonTextBlock
     tracks [{
       type=MovieSceneColorTrack
       display_name=Color and Opacity
       sections [{
         type=MovieSceneColorSection
         end_frame=2
         start_frame=0
         channels [
          Color.R {type=MovieSceneScriptingFloatChannel, keys [[0,0.6038],[2,0.007]]}
          Color.G {type=MovieSceneScriptingFloatChannel, keys [[0,0.8879],[2,0.017]]}
          Color.B {type=MovieSceneScriptingFloatChannel, keys [[0,1],[2,0.2]]}
          Color.A {type=MovieSceneScriptingFloatChannel}
         ]
        }]
      }]
    }
    Border_Background {
     id=#4
     possessed_class=CommonBorder
     tracks [{
       type=MovieSceneWidgetMaterialTrack
       sections [{
         type=MovieSceneParameterSection
         end_frame=2
         start_frame=0
         channels [
          BottomAngle {type=MovieSceneScriptingFloatChannel, keys [[0,0],[2,4]]}
          BottomOffset {type=MovieSceneScriptingFloatChannel, keys [[0,0.1],[2,0.03]]}
          LeftAngle {type=MovieSceneScriptingFloatChannel, keys [[0,0],[2,4]]}
          LeftOffset {type=MovieSceneScriptingFloatChannel, keys [[0,-0.1],[2,-0.05]]}
          Opacity {type=MovieSceneScriptingFloatChannel, keys [[0,0.5],[2,1]]}
          RightAngle {type=MovieSceneScriptingFloatChannel, keys [[0,0],[2,-4]]}
          RightOffset {type=MovieSceneScriptingFloatChannel, keys [[0,0.1],[2,0.04]]}
          StatusColorOffset {type=MovieSceneScriptingFloatChannel, keys [[0,-0.35],[2,-0.47]]}
          TopAngle {type=MovieSceneScriptingFloatChannel, keys [[0,0],[2,-2]]}
          TopOffset {type=MovieSceneScriptingFloatChannel, keys [[0,-0.1],[2,-0.05]]}
          MainColorR {type=MovieSceneScriptingFloatChannel, keys [[0,0.0075],[2,1]]}
          MainColorG {type=MovieSceneScriptingFloatChannel, keys [[0,0.0176],[2,1]]}
          MainColorB {type=MovieSceneScriptingFloatChannel, keys [[0,0.2086],[2,1]]}
          MainColorA {type=MovieSceneScriptingFloatChannel, const [0,2]}
         ]
        }]
      }]
    }
    Text_Rank {
     id=#5
     possessed_class=CommonTextBlock
     tracks [{
       type=MovieSceneColorTrack
       display_name=Color and Opacity
       sections [{
         type=MovieSceneColorSection
         end_frame=2
         start_frame=0
         channels [
          Color.R {type=MovieSceneScriptingFloatChannel, keys [[0,0.6038],[2,0.007]]}
          Color.G {type=MovieSceneScriptingFloatChannel, keys [[0,0.8879],[2,0.017]]}
          Color.B {type=MovieSceneScriptingFloatChannel, keys [[0,1],[2,0.2]]}
          Color.A {type=MovieSceneScriptingFloatChannel}
         ]
        }]
      }]
    }
    Text_Name {
     id=#6
     possessed_class=CommonTextBlock
     tracks [{
       type=MovieSceneColorTrack
       display_name=Color and Opacity
       sections [{
         type=MovieSceneColorSection
         end_frame=2
         start_frame=0
         channels [
          Color.R {type=MovieSceneScriptingFloatChannel, keys [[0,1],[2,0.007]]}
          Color.G {type=MovieSceneScriptingFloatChannel, keys [[0,1],[2,0.017]]}
          Color.B {type=MovieSceneScriptingFloatChannel, keys [[0,1],[2,0.2]]}
          Color.A {type=MovieSceneScriptingFloatChannel}
         ]
        }]
      }]
    }
   ]}
 }
 Hovered {
  binding_count=6
  length_seconds=0.1
  object=&self:Hovered
  sequence {bindings [
    Text_Date {
     id=#1
     possessed_class=CommonTextBlock
     tracks [{
       type=MovieSceneColorTrack
       display_name=Color and Opacity
       sections [{
         type=MovieSceneColorSection
         end_frame=2
         start_frame=0
         channels [
          Color.R {type=MovieSceneScriptingFloatChannel, keys [[0,0.6038],[2,0.007]]}
          Color.G {type=MovieSceneScriptingFloatChannel, keys [[0,0.8879],[2,0.017]]}
          Color.B {type=MovieSceneScriptingFloatChannel, keys [[0,1],[2,0.2]]}
          Color.A {type=MovieSceneScriptingFloatChannel}
         ]
        }]
      }]
    }
    Text_Length {
     id=#2
     possessed_class=CommonTextBlock
     tracks [{
       type=MovieSceneColorTrack
       display_name=Color and Opacity
       sections [{
         type=MovieSceneColorSection
         end_frame=2
         start_frame=0
         channels [
          Color.R {type=MovieSceneScriptingFloatChannel, keys [[0,0.6038],[2,0.007]]}
          Color.G {type=MovieSceneScriptingFloatChannel, keys [[0,0.8879],[2,0.017]]}
          Color.B {type=MovieSceneScriptingFloatChannel, keys [[0,1],[2,0.2]]}
          Color.A {type=MovieSceneScriptingFloatChannel}
         ]
        }]
      }]
    }
    Text_Eliminations {
     id=#3
     possessed_class=CommonTextBlock
     tracks [{
       type=MovieSceneColorTrack
       display_name=Color and Opacity
       sections [{
         type=MovieSceneColorSection
         end_frame=2
         start_frame=0
         channels [
          Color.R {type=MovieSceneScriptingFloatChannel, keys [[0,0.6038],[2,0.007]]}
          Color.G {type=MovieSceneScriptingFloatChannel, keys [[0,0.8879],[2,0.017]]}
          Color.B {type=MovieSceneScriptingFloatChannel, keys [[0,1],[2,0.2]]}
          Color.A {type=MovieSceneScriptingFloatChannel}
         ]
        }]
      }]
    }
    Text_Rank {
     id=#5
     possessed_class=CommonTextBlock
     tracks [{
       type=MovieSceneColorTrack
       display_name=Color and Opacity
       sections [{
         type=MovieSceneColorSection
         end_frame=2
         start_frame=0
         channels [
          Color.R {type=MovieSceneScriptingFloatChannel, keys [[0,0.6038],[2,0.007]]}
          Color.G {type=MovieSceneScriptingFloatChannel, keys [[0,0.8879],[2,0.017]]}
          Color.B {type=MovieSceneScriptingFloatChannel, keys [[0,1],[2,0.2]]}
          Color.A {type=MovieSceneScriptingFloatChannel}
         ]
        }]
      }]
    }
    Text_Name {
     id=#6
     possessed_class=CommonTextBlock
     tracks [{
       type=MovieSceneColorTrack
       display_name=Color and Opacity
       sections [{
         type=MovieSceneColorSection
         end_frame=2
         start_frame=0
         channels [
          Color.R {type=MovieSceneScriptingFloatChannel, keys [[0,1],[2,0.007]]}
          Color.G {type=MovieSceneScriptingFloatChannel, keys [[0,1],[2,0.017]]}
          Color.B {type=MovieSceneScriptingFloatChannel, keys [[0,1],[2,0.2]]}
          Color.A {type=MovieSceneScriptingFloatChannel}
         ]
        }]
      }]
    }
    Border_Background {
     id=#7
     possessed_class=CommonBorder
     tracks [{
       type=MovieSceneWidgetMaterialTrack
       sections [{
         type=MovieSceneParameterSection
         end_frame=2
         start_frame=0
         channels [
          BottomAngle {type=MovieSceneScriptingFloatChannel, keys [[0,0],[2,-2]]}
          BottomOffset {type=MovieSceneScriptingFloatChannel, keys [[0,0.1],[2,0.115]]}
          LeftAngle {type=MovieSceneScriptingFloatChannel, keys [[0,0],[2,-2]]}
          LeftOffset {type=MovieSceneScriptingFloatChannel, keys [[0,-0.1],[2,-0.075]]}
          Opacity {type=MovieSceneScriptingFloatChannel, keys [[0,0.5],[2,1]]}
          RightAngle {type=MovieSceneScriptingFloatChannel, keys [[0,0],[2,4]]}
          RightOffset {type=MovieSceneScriptingFloatChannel, keys [[0,0.1],[2,0.13]]}
          StatusColorOffset {type=MovieSceneScriptingFloatChannel, keys [[0,-0.35],[2,-0.4]]}
          TopAngle {type=MovieSceneScriptingFloatChannel, keys [[0,0],[2,0.5]]}
          MainColorR {type=MovieSceneScriptingFloatChannel, keys [[0,0.0075],[2,1]]}
          MainColorG {type=MovieSceneScriptingFloatChannel, keys [[0,0.0176],[2,1]]}
          MainColorB {type=MovieSceneScriptingFloatChannel, keys [[0,0.2086],[2,1]]}
          MainColorA {type=MovieSceneScriptingFloatChannel, const [0,2]}
         ]
        }]
      }]
    }
   ]}
 }
 Selected_Hovered {
  binding_count=6
  length_seconds=0.1
  object=&self:Selected_Hovered
  sequence {bindings [
    Text_Date {
     id=#1
     possessed_class=CommonTextBlock
     tracks [{
       type=MovieSceneColorTrack
       display_name=Color and Opacity
       sections [{
         type=MovieSceneColorSection
         end_frame=2
         start_frame=0
         channels [
          Color.R {type=MovieSceneScriptingFloatChannel, const [0.007,2]}
          Color.G {type=MovieSceneScriptingFloatChannel, const [0.017,2]}
          Color.B {type=MovieSceneScriptingFloatChannel, const [0.2,2]}
          Color.A {type=MovieSceneScriptingFloatChannel}
         ]
        }]
      }]
    }
    Text_Length {
     id=#2
     possessed_class=CommonTextBlock
     tracks [{
       type=MovieSceneColorTrack
       display_name=Color and Opacity
       sections [{
         type=MovieSceneColorSection
         end_frame=2
         start_frame=0
         channels [
          Color.R {type=MovieSceneScriptingFloatChannel, const [0.007,2]}
          Color.G {type=MovieSceneScriptingFloatChannel, const [0.017,2]}
          Color.B {type=MovieSceneScriptingFloatChannel, const [0.2,2]}
          Color.A {type=MovieSceneScriptingFloatChannel}
         ]
        }]
      }]
    }
    Text_Eliminations {
     id=#3
     possessed_class=CommonTextBlock
     tracks [{
       type=MovieSceneColorTrack
       display_name=Color and Opacity
       sections [{
         type=MovieSceneColorSection
         end_frame=2
         start_frame=0
         channels [
          Color.R {type=MovieSceneScriptingFloatChannel, const [0.007,2]}
          Color.G {type=MovieSceneScriptingFloatChannel, const [0.017,2]}
          Color.B {type=MovieSceneScriptingFloatChannel, const [0.2,2]}
          Color.A {type=MovieSceneScriptingFloatChannel}
         ]
        }]
      }]
    }
    Border_Background {
     id=#8
     possessed_class=CommonBorder
     tracks [{
       type=MovieSceneWidgetMaterialTrack
       sections [{
         type=MovieSceneParameterSection
         end_frame=2
         start_frame=0
         channels [
          BottomAngle {type=MovieSceneScriptingFloatChannel, keys [[0,-2],[2,4]]}
          BottomOffset {type=MovieSceneScriptingFloatChannel, keys [[0,0.115],[2,0.03]]}
          LeftAngle {type=MovieSceneScriptingFloatChannel, keys [[0,-2],[2,4]]}
          LeftOffset {type=MovieSceneScriptingFloatChannel, keys [[0,-0.075],[2,-0.05]]}
          Opacity {type=MovieSceneScriptingFloatChannel, const [1,2]}
          RightAngle {type=MovieSceneScriptingFloatChannel, keys [[0,4],[2,-4]]}
          RightOffset {type=MovieSceneScriptingFloatChannel, keys [[0,0.13],[2,0.04]]}
          StatusColorOffset {type=MovieSceneScriptingFloatChannel, keys [[0,-0.4],[2,-0.47]]}
          TopAngle {type=MovieSceneScriptingFloatChannel, keys [[0,0.5],[2,-2]]}
          TopOffset {type=MovieSceneScriptingFloatChannel, keys [[0,-0.1],[2,-0.05]]}
          MainColorR {type=MovieSceneScriptingFloatChannel, const [1,2]}
          MainColorG {type=MovieSceneScriptingFloatChannel, const [1,2]}
          MainColorB {type=MovieSceneScriptingFloatChannel, const [1,2]}
          MainColorA {type=MovieSceneScriptingFloatChannel, const [0,2]}
         ]
        }]
      }]
    }
    Text_Rank {
     id=#5
     possessed_class=CommonTextBlock
     tracks [{
       type=MovieSceneColorTrack
       display_name=Color and Opacity
       sections [{
         type=MovieSceneColorSection
         end_frame=2
         start_frame=0
         channels [
          Color.R {type=MovieSceneScriptingFloatChannel, const [0.007,2]}
          Color.G {type=MovieSceneScriptingFloatChannel, const [0.017,2]}
          Color.B {type=MovieSceneScriptingFloatChannel, const [0.2,2]}
          Color.A {type=MovieSceneScriptingFloatChannel}
         ]
        }]
      }]
    }
    Text_Name {
     id=#6
     possessed_class=CommonTextBlock
     tracks [{
       type=MovieSceneColorTrack
       display_name=Color and Opacity
       sections [{
         type=MovieSceneColorSection
         end_frame=2
         start_frame=0
         channels [
          Color.R {type=MovieSceneScriptingFloatChannel, const [0.007,2]}
          Color.G {type=MovieSceneScriptingFloatChannel, const [0.017,2]}
          Color.B {type=MovieSceneScriptingFloatChannel, const [0.2,2]}
          Color.A {type=MovieSceneScriptingFloatChannel}
         ]
        }]
      }]
    }
   ]}
 }
]
graphs [EventGraph {nodes [
   n1 = Event Handle Replay State Set!() -> (OutputDelegate, then, bSaved:bool, bOld:bool, bCorrupt:bool) {is=title K2Node_Event}
   > n2 = GetDynamicMaterial(self:Border Object Reference=Get Border_Background() -> (Border_Background:Common Border Object Reference))
   > SetVectorParameterValue(
    self=@n2
    ParameterName=Status Color
    Value=Select(
     Option 0:Linear Color Structure=Select(Option 0=(R=0.9301,G=1,B=0.0097,A=1), Option 1=(R=0.0331,G=0.855,B=1,A=1), Index=@n1.bSaved)
     Option 1:Linear Color Structure=(R=0.7913,G=0.0086,B=0.1845,A=1)
     Index=OR Boolean(A=@n1.bOld, B=@n1.bCorrupt)
    ) -> (ReturnValue:Linear Color Structure)
   )
   Event On Selected()
   > PlayAnimationForward(InAnimation=Select(
     Option 0:Widget Animation Object Reference=Get Selected_Direct() -> (Selected_Direct:Widget Animation Object Reference)
     Option 1:Widget Animation Object Reference=Get Selected_Hovered() -> (Selected_Hovered:Widget Animation Object Reference)
     Index=IsHovered()
    ) -> (ReturnValue:Widget Animation Object Reference))
   Event On Deselected()
   > PlayAnimationReverse(InAnimation=Get Selected_Direct())
   Event On Hovered()
   > n3 = Branch(Condition=GetSelected!(self:Common Button Base Object Reference) -> (ReturnValue:bool))
   PlayAnimationForward(execute=@n3.else, InAnimation=Get Hovered() -> (Hovered:Widget Animation Object Reference))
   Event On Unhovered()
   > n4 = Branch(Condition=GetSelected())
   PlayAnimationReverse(execute=@n4.else, InAnimation=Get Hovered())
  ]}]
variables []
widgets {
 info {inheritedWidgetCount=7, parentClass=/Script/FortniteUI.AthenaReplayBrowserEntryWidget, rootWidgetClass=&SizeBox, widgetCount=16}
 widgets [
  RowHeightSB {widgetClassPath=&SizeBox, widget=RowHeightSB, properties {bOverride_HeightOverride=true, heightOverride=58}}
  Border_Background {
   widgetClassPath=&CommonBorder
   bIsVariable=true
   parent=RowHeightSB
   widget=Border_Background
   properties {
    horizontalAlignment=HAlign_Left
    background {
     bIsDynamicallyLoaded=false
     drawAs=Box
     imageType=NoImage
     mirroring=NoMirror
     resourceName=None
     resourceObject=/Game/Athena/UI/ReplayBrowser/M_Ui_ReplayBrowserRow_Inst
     tiling=NoTile
     imageSize=<40 40>
     margin=<ltrb 0.5 0 0.5 0>
     outlineSettings {
      bUseBrushTransparency=false
      roundingType=HalfHeightRadius
      width=0
      color {colorUseRule=UseColor_Specified, specifiedColor=<rgba 0 0 0 0>}
      cornerRadii=<0 0 0 0>
     }
     tintColor {colorUseRule=UseColor_Specified, specifiedColor=<rgba 1 1 1 1>}
     uVRegion {bIsValid=false, max=<0 0>, min=<0 0>}
    }
    padding=<ltrb 0 0 0 0>
   }
   slot {class=SizeBoxSlot}
  }
  HorizontalBox_0 {
   widgetClassPath=&HorizontalBox
   parent=Border_Background
   widget=HorizontalBox_0
   slot {class=BorderSlot, properties {horizontalAlignment=HAlign_Left, padding=<ltrb 0 0 0 0>}}
  }
  NameSB {
   widgetClassPath=&SizeBox
   parent=HorizontalBox_0
   widget=NameSB
   properties {bOverride_WidthOverride=true, widthOverride=720}
   slot {class=HorizontalBoxSlot, properties {padding=<ltrb 32 0 8 0>}}
  }
  HorizontalBox_1 {
   widgetClassPath=&HorizontalBox
   parent=NameSB
   widget=HorizontalBox_1
   slot {class=SizeBoxSlot, properties {verticalAlignment=VAlign_Center}}
  }
  DownloadIcon {
   widgetClassPath=&Image
   bInherited=true
   bIsVariable=true
   parent=HorizontalBox_1
   widget=DownloadIcon
   properties {
    visibility=Collapsed
    brush {
     bIsDynamicallyLoaded=false
     drawAs=Image
     imageType=NoImage
     mirroring=NoMirror
     resourceName=None
     resourceObject=/Game/UI/Foundation/Toasts/Materials/T_UI_icon_downloadmessage
     tiling=NoTile
     imageSize=<32 32>
     margin=<ltrb 0 0 0 0>
     outlineSettings {
      bUseBrushTransparency=false
      roundingType=HalfHeightRadius
      width=0
      color {colorUseRule=UseColor_Specified, specifiedColor=<rgba 0 0 0 0>}
      cornerRadii=<0 0 0 0>
     }
     tintColor {colorUseRule=UseColor_Specified, specifiedColor=<rgba 1 1 1 1>}
     uVRegion {bIsValid=false, max=<0 0>, min=<0 0>}
    }
   }
   slot {class=HorizontalBoxSlot, properties {horizontalAlignment=HAlign_Center, verticalAlignment=VAlign_Center, padding=<ltrb 0 0 8 0>}}
  }
  Text_Name {
   widgetClassPath=&CommonTextBlock
   bInherited=true
   parent=HorizontalBox_1
   widget=Text_Name
   properties {
    scrollStyle=&TextScrollStyle-Base_C
    text=Replay Name
    font {
     bForceMonospaced=false
     bMaterialIsStencil=false
     fontMaterial=null
     fontObject=&BurbankSmall
     letterSpacing=0
     monospacedWidth=1
     size=18
     skewAmount=0
     typefaceFontName=Bold
     outlineSettings {
      bApplyOutlineToDropShadows=false
      bMiteredCorners=false
      bSeparateFillAlpha=false
      outlineMaterial=null
      outlineSize=0
      outlineColor=<rgba 0 0 0 1>
     }
    }
    shadowOffset=<0 0>
   }
   slot {class=HorizontalBoxSlot}
  }
  DateSB {
   widgetClassPath=&SizeBox
   parent=HorizontalBox_0
   widget=DateSB
   properties {bOverride_WidthOverride=true, widthOverride=256}
   slot {class=HorizontalBoxSlot, properties {padding=<ltrb 8 0 8 0>}}
  }
  Text_Date {
   widgetClassPath=&CommonTextBlock
   bInherited=true
   parent=DateSB
   widget=Text_Date
   properties {
    justification=Center
    text="88/88/8888, 88:88 WW"
    colorAndOpacity {colorUseRule=UseColor_Specified, specifiedColor=<rgba 0.6038 0.8879 1 1>}
    font {
     bForceMonospaced=false
     bMaterialIsStencil=false
     fontMaterial=null
     fontObject=&BurbankSmall
     letterSpacing=0
     monospacedWidth=1
     size=18
     skewAmount=0
     typefaceFontName=Bold
     outlineSettings {
      bApplyOutlineToDropShadows=false
      bMiteredCorners=false
      bSeparateFillAlpha=false
      outlineMaterial=null
      outlineSize=0
      outlineColor=<rgba 0 0 0 1>
     }
    }
    shadowOffset=<0 0>
   }
   slot {class=SizeBoxSlot, properties {verticalAlignment=VAlign_Center, padding=<ltrb 5 0 5 0>}}
  }
  LengthSB {
   widgetClassPath=&SizeBox
   parent=HorizontalBox_0
   widget=LengthSB
   properties {bOverride_WidthOverride=true, widthOverride=200}
   slot {class=HorizontalBoxSlot, properties {padding=<ltrb 8 0 8 0>}}
  }
  Text_Length {
   widgetClassPath=&CommonTextBlock
   bInherited=true
   parent=LengthSB
   widget=Text_Length
   properties {
    justification=Center
    text=88:88:88
    colorAndOpacity {colorUseRule=UseColor_Specified, specifiedColor=<rgba 0.6038 0.8879 1 1>}
    font {
     bForceMonospaced=false
     bMaterialIsStencil=false
     fontMaterial=null
     fontObject=&BurbankSmall
     letterSpacing=0
     monospacedWidth=1
     size=18
     skewAmount=0
     typefaceFontName=Bold
     outlineSettings {
      bApplyOutlineToDropShadows=false
      bMiteredCorners=false
      bSeparateFillAlpha=false
      outlineMaterial=null
      outlineSize=0
      outlineColor=<rgba 0 0 0 1>
     }
    }
    shadowOffset=<0 0>
   }
   slot {class=SizeBoxSlot, properties {horizontalAlignment=HAlign_Center, verticalAlignment=VAlign_Center, padding=<ltrb 5 0 5 0>}}
  }
  FinishSB {
   widgetClassPath=&SizeBox
   parent=HorizontalBox_0
   widget=FinishSB
   properties {bOverride_WidthOverride=true, widthOverride=200}
   slot {class=HorizontalBoxSlot, properties {padding=<ltrb 8 0 8 0>}}
  }
  Text_Rank {
   widgetClassPath=&CommonTextBlock
   bInherited=true
   parent=FinishSB
   widget=Text_Rank
   properties {
    justification=Center
    text=888/888
    colorAndOpacity {colorUseRule=UseColor_Specified, specifiedColor=<rgba 0.6038 0.8879 1 1>}
    font {
     bForceMonospaced=false
     bMaterialIsStencil=false
     fontMaterial=null
     fontObject=&BurbankSmall
     letterSpacing=0
     monospacedWidth=1
     size=18
     skewAmount=0
     typefaceFontName=Bold
     outlineSettings {
      bApplyOutlineToDropShadows=false
      bMiteredCorners=false
      bSeparateFillAlpha=false
      outlineMaterial=null
      outlineSize=0
      outlineColor=<rgba 0 0 0 1>
     }
    }
    shadowOffset=<0 0>
   }
   slot {class=SizeBoxSlot, properties {horizontalAlignment=HAlign_Center, verticalAlignment=VAlign_Center, padding=<ltrb 5 0 5 0>}}
  }
  EliminationsSB {
   widgetClassPath=&SizeBox
   parent=HorizontalBox_0
   widget=EliminationsSB
   properties {bOverride_WidthOverride=true, widthOverride=200}
   slot {class=HorizontalBoxSlot, properties {padding=<ltrb 8 0 8 0>}}
  }
  Text_Eliminations {
   widgetClassPath=&CommonTextBlock
   bInherited=true
   parent=EliminationsSB
   widget=Text_Eliminations
   properties {
    justification=Center
    text=888
    colorAndOpacity {colorUseRule=UseColor_Specified, specifiedColor=<rgba 0.6038 0.8879 1 1>}
    font {
     bForceMonospaced=false
     bMaterialIsStencil=false
     fontMaterial=null
     fontObject=&BurbankSmall
     letterSpacing=0
     monospacedWidth=1
     size=18
     skewAmount=0
     typefaceFontName=Bold
     outlineSettings {
      bApplyOutlineToDropShadows=false
      bMiteredCorners=false
      bSeparateFillAlpha=false
      outlineMaterial=null
      outlineSize=0
      outlineColor=<rgba 0 0 0 1>
     }
    }
    shadowOffset=<0 0>
   }
   slot {class=SizeBoxSlot, properties {horizontalAlignment=HAlign_Center, verticalAlignment=VAlign_Center, padding=<ltrb 5 0 5 0>}}
  }
  InputActionWidget {widgetClassPath=&CommonActionWidget, bInherited=true, widget=null}
 ]
}
