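# Leaderboard throbber widget, a FortUserWidget that shows a busy indicator with a message while a leaderboard updates.
# A BackgroundBlur fills the root overlay behind a centered soft-shadow panel holding a material-driven throbber image and a CommonTextBlock.
# PreConstruct copies the UpdatingText variable (default "Updating") into the text block, replacing the designer placeholder "Updating Leaderboard".
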WidgetBlueprint /Game/UI/Foundation/Throbber/LeaderboardThrobber
graphs [EventGraph {nodes [
   Event PreConstruct()
   > "SetText (Text)"(self=Get TextMessage() -> (TextMessage:Common Text Object Reference), InText=Get UpdatingText() -> (UpdatingText:Text))
  ]}]
variables [UpdatingText {type=string, default=Updating}]
widgets {
 info {inheritedWidgetCount=0, parentClass=/Script/FortniteUI.FortUserWidget, rootWidgetClass=&Overlay, widgetCount=9}
 widgets [
  Overlay_1 {widgetClassPath=&Overlay, widget=Overlay_1}
  BackgroundBlur_90 {
   widgetClassPath=&BackgroundBlur
   parent=Overlay_1
   widget=BackgroundBlur_90
   properties {blurStrength=5}
   slot {class=OverlaySlot, properties {horizontalAlignment=HAlign_Fill, verticalAlignment=VAlign_Fill}}
  }
  Image_22 {
   widgetClassPath=&Image
   bIsVariable=true
   parent=Overlay_1
   widget=Image_22
   properties {
    brush {
     bIsDynamicallyLoaded=false
     drawAs=Image
     imageType=NoImage
     mirroring=NoMirror
     resourceName=None
     resourceObject=null
     tiling=NoTile
     imageSize=<32 32>
     margin=<ltrb 0 0 0 0>
     outlineSettings {
      bUseBrushTransparency=false
      roundingType=HalfHeightRadius
      width=0
      color {colorUseRule=UseColor_Specified, specifiedColor=<rgba 0 0 0 0>}
      cornerRadii=<0 0 0 0>
     }
     tintColor {colorUseRule=UseColor_Specified, specifiedColor=<rgba 0.0075 0.0176 0.2086 1>}
     uVRegion {bIsValid=false, max=<0 0>, min=<0 0>}
    }
    colorAndOpacity=<rgba 1 1 1 0>
   }
   slot {class=OverlaySlot, properties {horizontalAlignment=HAlign_Fill, verticalAlignment=VAlign_Fill}}
  }
  Overlay_0 {
   widgetClassPath=&Overlay
   parent=Overlay_1
   widget=Overlay_0
   slot {class=OverlaySlot, properties {horizontalAlignment=HAlign_Center, verticalAlignment=VAlign_Center}}
  }
  Image_0 {
   widgetClassPath=&Image
   bIsVariable=true
   parent=Overlay_0
   widget=Image_0
   properties {
    brush {
     bIsDynamicallyLoaded=false
     drawAs=Image
     imageType=NoImage
     mirroring=NoMirror
     resourceName=None
     resourceObject=/FortUILibrary/Styles/Border/Materials/M_Border_SoftShadow_DarkBLue
     tiling=NoTile
     imageSize=<32 32>
     margin=<ltrb 0 0 0 0>
     outlineSettings {
      bUseBrushTransparency=false
      roundingType=HalfHeightRadius
      width=0
      color {colorUseRule=UseColor_Specified, specifiedColor=<rgba 0 0 0 0>}
      cornerRadii=<0 0 0 0>
     }
     tintColor {colorUseRule=UseColor_Specified, specifiedColor=<rgba 1 1 1 1>}
     uVRegion {bIsValid=false, max=<0 0>, min=<0 0>}
    }
    colorAndOpacity=<rgba 1 1 1 0.33>
   }
   slot {class=OverlaySlot, properties {horizontalAlignment=HAlign_Fill, verticalAlignment=VAlign_Fill}}
  }
  HorizontalBox_0 {
   widgetClassPath=&HorizontalBox
   parent=Overlay_0
   widget=HorizontalBox_0
   slot {class=OverlaySlot, properties {horizontalAlignment=HAlign_Fill, verticalAlignment=VAlign_Fill, padding=<ltrb 96 64 96 64>}}
  }
  SizeBox_5069 {
   widgetClassPath=&SizeBox
   parent=HorizontalBox_0
   widget=SizeBox_5069
   properties {bOverride_HeightOverride=true, bOverride_WidthOverride=true, heightOverride=96, widthOverride=96}
   slot {class=HorizontalBoxSlot}
  }
  ProgressSpinner {
   widgetClassPath=&Image
   bIsVariable=true
   parent=SizeBox_5069
   widget=ProgressSpinner
   properties {brush {
     bIsDynamicallyLoaded=false
     drawAs=Image
     imageType=NoImage
     mirroring=NoMirror
     resourceName=None
     resourceObject=/FortUILibrary/UIKit/Throbber/MI_UIKit_Monochrome_Throbber
     tiling=NoTile
     imageSize=<64 64>
     margin=<ltrb 0 0 0 0>
     outlineSettings {
      bUseBrushTransparency=false
      roundingType=HalfHeightRadius
      width=0
      color {colorUseRule=UseColor_Specified, specifiedColor=<rgba 0 0 0 0>}
      cornerRadii=<0 0 0 0>
     }
     tintColor {colorUseRule=UseColor_Specified, specifiedColor=<rgba 1 1 1 1>}
     uVRegion {bIsValid=false, max=<0 0>, min=<0 0>}
    }}
   slot {class=SizeBoxSlot, properties {horizontalAlignment=HAlign_Center, verticalAlignment=VAlign_Center}}
  }
  TextMessage {
   widgetClassPath=&CommonTextBlock
   bIsVariable=true
   parent=HorizontalBox_0
   widget=TextMessage
   properties {
    justification=Center
    text=Updating Leaderboard
    colorAndOpacity {colorUseRule=UseColor_Specified, specifiedColor=<rgba 0.2159 0.6584 1 1>}
    font {
     bForceMonospaced=false
     bMaterialIsStencil=false
     fontMaterial=null
     fontObject=&BurbankSmall
     letterSpacing=0
     monospacedWidth=1
     size=24
     skewAmount=0
     typefaceFontName=Bold
     outlineSettings {
      bApplyOutlineToDropShadows=false
      bMiteredCorners=false
      bSeparateFillAlpha=false
      outlineMaterial=null
      outlineSize=0
      outlineColor=<rgba 0 0 0 1>
     }
    }
    shadowOffset=<0 0>
   }
   slot {class=HorizontalBoxSlot, properties {horizontalAlignment=HAlign_Center, verticalAlignment=VAlign_Center}}
  }
 ]
}
