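# Replay browser tab widget, built on AthenaReplayBrowserScreenBase, for browsing and managing saved Fortnite replays.
# A header row and replay list sit above initially collapsed, input-action-bound Play, Delete, Rename and Open Folder buttons, with a rename dialog overlay.
# OnActionStateChanged starts a short timer to delay the loading throbber while an action runs, then clears it and collapses the throbber when none remains, so brief actions avoid flashing it.
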WidgetBlueprint /FortniteCoreUI/UI/ReplayBrowser/AthenaReplayBrowserTab
graphs [EventGraph {nodes [
   ShowLoadingThrobber() {is=title K2Node_CustomEvent}
   > SetVisibility(self=Get LoadingThrobber() -> (LoadingThrobber:Leaderboard Throbber Object Reference), InVisibility=SelfHitTestInvisible)
   n1 = Event OnActionStateChanged!() -> (OutputDelegate, then, bHasAction:bool) {is=title K2Node_Event}
   > n2 = Branch(Condition=@n1.bHasAction)
   > n3 = Branch(Condition=IsVisible(self=Get LoadingThrobber()))
   n4 = Set Timer by Event(execute=@n3.else, Delegate=Create Event(), Time=.25, bLooping=false)
   > Set LoadingTimerHandle(LoadingTimerHandle:Timer Handle Structure=@n4) -> (Output_Get:Timer Handle Structure)
   Clear and Invalidate Timer by Handle(execute=@n2.else, Handle=Get LoadingTimerHandle() -> (LoadingTimerHandle:Timer Handle Structure))
   > SetVisibility(self=Get LoadingThrobber(), InVisibility=Collapsed)
  ]}]
variables [LoadingTimerHandle {type=TimerHandle}]
widgets {
 info {inheritedWidgetCount=6, parentClass=/Script/FortniteUI.AthenaReplayBrowserScreenBase, rootWidgetClass=&Overlay, widgetCount=17}
 widgets [
  Overlay_3 {widgetClassPath=&Overlay, widget=Overlay_3}
  Image_109 {
   widgetClassPath=&Image
   parent=Overlay_3
   widget=Image_109
   properties {
    visibility=SelfHitTestInvisible
    brush {
     bIsDynamicallyLoaded=false
     drawAs=Image
     imageType=NoImage
     mirroring=NoMirror
     resourceName=None
     resourceObject=/Game/Athena/UI/Frontend/Art/M_UI_RadialGradientMaster_Inst
     tiling=NoTile
     imageSize=<32 32>
     margin=<ltrb 0 0 0 0>
     outlineSettings {
      bUseBrushTransparency=false
      roundingType=HalfHeightRadius
      width=0
      color {colorUseRule=UseColor_Specified, specifiedColor=<rgba 0 0 0 0>}
      cornerRadii=<0 0 0 0>
     }
     tintColor {colorUseRule=UseColor_Specified, specifiedColor=<rgba 1 1 1 1>}
     uVRegion {bIsValid=false, max=<0 0>, min=<0 0>}
    }
   }
   slot {class=OverlaySlot, properties {horizontalAlignment=HAlign_Fill, verticalAlignment=VAlign_Fill, padding=<ltrb -1 -1 -1 0>}}
  }
  SafeZone_0 {
   widgetClassPath=&SafeZone
   parent=Overlay_3
   widget=SafeZone_0
   properties {padBottom=false, padLeft=false, padRight=false}
   slot {class=OverlaySlot, properties {horizontalAlignment=HAlign_Fill, verticalAlignment=VAlign_Fill, padding=<ltrb 0 0 0 80>}}
  }
  Overlay_0 {widgetClassPath=&Overlay, parent=SafeZone_0, widget=Overlay_0, slot {class=SafeZoneSlot}}
  SafeZone_2 {
   widgetClassPath=&SafeZone
   bIsVariable=true
   parent=Overlay_0
   widget=SafeZone_2
   properties {padBottom=false, padRight=false, padTop=false}
   slot {class=OverlaySlot, properties {padding=<ltrb 40 40 32 0>}}
  }
  CommonTextBlock_6 {
   widgetClassPath=&CommonTextBlock
   parent=SafeZone_2
   widget=CommonTextBlock_6
   properties {
    text=Replays
    textTransformPolicy=ToUpper
    font {
     bForceMonospaced=false
     bMaterialIsStencil=false
     fontMaterial=/FortniteCoreUI/UI/Season/Art/M_UI_AthenaSeason_BattlePass_Text
     fontObject=&Burbank
     letterSpacing=0
     monospacedWidth=1
     size=128
     skewAmount=0
     typefaceFontName=BurbankRegular-Black
     outlineSettings {
      bApplyOutlineToDropShadows=false
      bMiteredCorners=false
      bSeparateFillAlpha=false
      outlineMaterial=null
      outlineSize=0
      outlineColor=<rgba 0 0 0 1>
     }
    }
    margin=<ltrb 0 -40 0 0>
    renderTransform {angle=0, scale=<1 1>, shear=<-12 0>, translation=<0 0>}
    shadowOffset=<0 0>
   }
   slot {class=SafeZoneSlot}
  }
  VerticalBox_0 {
   widgetClassPath=&VerticalBox
   parent=Overlay_0
   widget=VerticalBox_0
   slot {class=OverlaySlot, properties {horizontalAlignment=HAlign_Fill, verticalAlignment=VAlign_Fill, padding=<ltrb 100 128 100 20>}}
  }
  HeaderRow {
   widgetClassPath=/FortniteCoreUI/Athena/UI/ReplayBrowser/AthenaReplayBrowserHeaderRow
   parent=VerticalBox_0
   widget=HeaderRow
   properties {flowDirectionPreference=Inherit}
   slot {class=VerticalBoxSlot}
  }
  Overlay_1 {
   widgetClassPath=&Overlay
   parent=VerticalBox_0
   widget=Overlay_1
   slot {class=VerticalBoxSlot, properties {size {sizeRule=Fill, value=1}}}
  }
  ListView_Replay {
   widgetClassPath=&CommonListView
   bInherited=true
   parent=Overlay_1
   widget=ListView_Replay
   properties {
    entryWidgetClass=/FortniteCoreUI/Athena/UI/ReplayBrowser/AthenaReplayBrowserRow.AthenaReplayBrowserRow_C
    navigation=&self:WidgetTree.ListView_Replay.WidgetNavigation_0
    numDesignerPreviewEntries=15
   }
   slot {class=OverlaySlot, properties {horizontalAlignment=HAlign_Fill, verticalAlignment=VAlign_Fill}}
  }
  LoadingThrobber {
   widgetClassPath=/Game/UI/Foundation/Throbber/LeaderboardThrobber
   bIsVariable=true
   parent=Overlay_1
   widget=LoadingThrobber
   properties {updatingText=Loading, visibility=Collapsed}
   slot {class=OverlaySlot, properties {horizontalAlignment=HAlign_Fill, verticalAlignment=VAlign_Fill}}
  }
  HorizontalBox_1 {
   widgetClassPath=&HorizontalBox
   parent=VerticalBox_0
   widget=HorizontalBox_1
   slot {class=VerticalBoxSlot, properties {horizontalAlignment=HAlign_Center, padding=<ltrb 0 20 0 0>, size {sizeRule=Fill, value=0.1}}}
  }
  Button_PlayReplay {
   widgetClassPath=&WBP_UIKit_Button_Regular
   bInherited=true
   bIsVariable=true
   parent=HorizontalBox_1
   widget=Button_PlayReplay
   properties {
    bIsFocusable=false
    minHeight=54
    minWidth=256
    text=Play
    visibility=Collapsed
    lockedPressedSlateSoundOverride {resourceObject=/Game/Sounds/Fort_UI_Sounds/FrontEnd/CareerPage/Fortnite_UI_Career_Button_Generic_01_Cue}
    pressedSlateSoundOverride {resourceObject=&Fortnite_UI_Career_Button_Generic_01_Cue}
    selectedPressedSlateSoundOverride {resourceObject=&Fortnite_UI_Career_Button_Generic_01_Cue}
    triggeringInputAction {dataTable=/FortUILibrary/Input/ReplayInputActionDataTable, rowName=ReplayBrowserPlay}
   }
   slot {class=HorizontalBoxSlot, properties {padding=<ltrb 12 0 12 0>}}
  }
  Button_DeleteReplay {
   widgetClassPath=&WBP_UIKit_Button_Regular
   bInherited=true
   bIsVariable=true
   parent=HorizontalBox_1
   widget=Button_DeleteReplay
   properties {
    bIsFocusable=false
    minWidth=256
    text=Delete
    visibility=Collapsed
    lockedPressedSlateSoundOverride {resourceObject=&Fortnite_UI_Career_Button_Generic_01_Cue}
    pressedSlateSoundOverride {resourceObject=&Fortnite_UI_Career_Button_Generic_01_Cue}
    selectedPressedSlateSoundOverride {resourceObject=&Fortnite_UI_Career_Button_Generic_01_Cue}
    triggeringInputAction {dataTable=&ReplayInputActionDataTable, rowName=ReplayBrowserDelete}
   }
   slot {class=HorizontalBoxSlot, properties {padding=<ltrb 12 0 12 0>}}
  }
  Button_RenameReplay {
   widgetClassPath=&WBP_UIKit_Button_Regular
   bInherited=true
   bIsVariable=true
   parent=HorizontalBox_1
   widget=Button_RenameReplay
   properties {
    bIsFocusable=false
    minWidth=256
    text=Rename and save
    visibility=Collapsed
    lockedPressedSlateSoundOverride {resourceObject=&Fortnite_UI_Career_Button_Generic_01_Cue}
    pressedSlateSoundOverride {resourceObject=&Fortnite_UI_Career_Button_Generic_01_Cue}
    selectedPressedSlateSoundOverride {resourceObject=&Fortnite_UI_Career_Button_Generic_01_Cue}
    triggeringInputAction {dataTable=&ReplayInputActionDataTable, rowName=ReplayBrowserRename}
   }
   slot {class=HorizontalBoxSlot, properties {padding=<ltrb 12 0 12 0>}}
  }
  Button_OpenReplayFolder {
   widgetClassPath=&WBP_UIKit_Button_Regular
   bInherited=true
   bIsVariable=true
   parent=HorizontalBox_1
   widget=Button_OpenReplayFolder
   properties {
    bIsFocusable=false
    minWidth=256
    text=Open replay folder
    visibility=Collapsed
    lockedPressedSlateSoundOverride {resourceObject=&Fortnite_UI_Career_Button_Generic_01_Cue}
    pressedSlateSoundOverride {resourceObject=&Fortnite_UI_Career_Button_Generic_01_Cue}
    selectedPressedSlateSoundOverride {resourceObject=&Fortnite_UI_Career_Button_Generic_01_Cue}
    triggeringInputAction {dataTable=&ReplayInputActionDataTable, rowName=ReplayBrowserOpenFolder}
   }
   slot {class=HorizontalBoxSlot, properties {padding=<ltrb 12 0 12 0>}}
  }
  RenameDialog {
   widgetClassPath=/FortniteCoreUI/Athena/UI/ReplayBrowser/AthenaReplayBrowserRenameDialog
   bInherited=true
   parent=Overlay_0
   widget=RenameDialog
   properties {visibility=Collapsed}
   slot {class=OverlaySlot, properties {horizontalAlignment=HAlign_Fill, verticalAlignment=VAlign_Fill}}
  }
 ]
}
